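# Physics asset for the dynamic belt on the male medium Slither player body cosmetic.
# Left and right three-bone capsule chains hang from small master bodies, which join a central belt master through fully free constraints.
# Chain links have tight twist limits, with much wider swing on the right than the left.
# Unconstrained capsules on both thighs and a large front-skirt collision bone provide body collision volumes; belt shape names carry over from a skirt setup.
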PhysicsAsset /BRCosmetics/Characters/Player/Male/Medium/Bodies/M_MED_Slither/Meshes/M_MED_Slither_dyn_belt
bodies {
 dyn_belt_MSTR [{shapeType=Capsule, length=0.505, radius=0.505, shapeName=dyn_skirt_bk_r_MSTR_capsule, rotation=<rot 0 -90 0>}]
 dyn_belt_l_01 [{shapeType=Capsule, length=2, radius=2, shapeName=dyn_skirt_bk_l_01_capsule, center=<3 0 0>, rotation=<rot 90 0 0>}]
 dyn_belt_l_02 [{shapeType=Capsule, length=2, radius=2.5, shapeName=dyn_skirt_bk_l_01_capsule, center=<3 0 0>, rotation=<rot 90 0 0>}]
 dyn_belt_l_03 [{shapeType=Capsule, length=3.75, radius=2.5, shapeName=dyn_skirt_bk_l_02_capsule, center=<3 0 0>, rotation=<rot 90 0 0>}]
 dyn_belt_l_MSTR [{shapeType=Capsule, length=0.505, radius=0.505, shapeName=dyn_skirt_bk_r_MSTR_capsule, rotation=<rot 0 -90 0>}]
 dyn_belt_r_01 [{shapeType=Capsule, length=2.75, radius=2, shapeName=dyn_skirt_bk_l_01_capsule, center=<4 0 0>, rotation=<rot 90 0 0>}]
 dyn_belt_r_02 [{shapeType=Capsule, length=3.5, radius=2, shapeName=dyn_skirt_bk_l_01_capsule, center=<4 0 0>, rotation=<rot 90 0 0>}]
 dyn_belt_r_03 [{shapeType=Capsule, length=3.5, radius=2, shapeName=dyn_skirt_bk_l_02_capsule, center=<4 0 0>, rotation=<rot 90 0 0>}]
 dyn_belt_r_MSTR [{shapeType=Capsule, length=0.505, radius=0.505, shapeName=dyn_skirt_bk_r_MSTR_capsule, rotation=<rot 0 -90 0>}]
 dyn_skirt_ft_collision [{shapeType=Capsule, length=18.5, radius=32.7732, shapeName=pelvis_capsule, center=<5.4999 -22.322 -0.0004>}]
 thigh_twist_01_l [{
   shapeType=Capsule
   length=23.682
   radius=10.9857
   shapeName=thigh_twist_01_l_capsule
   center=<2.5924 -1.1428 1.2859>
   rotation=<rot 88.1297 0 0.1028>
  }]
 thigh_twist_01_r [{
   shapeType=Capsule
   length=26.432
   radius=10.4857
   shapeName=thigh_twist_01_l_capsule
   center=<-2.5655 -0.2101 -1.9056>
   rotation=<rot 88.1331 0 -178.155>
  }]
}
constraints [
 {bone1Name=dyn_belt_r_01, bone2Name=dyn_belt_r_MSTR, swing1LimitDegrees=20, swing2LimitDegrees=20, twistLimitDegrees=2}
 {bone1Name=dyn_belt_r_02, bone2Name=dyn_belt_r_01, swing1LimitDegrees=20, swing2LimitDegrees=35, twistLimitDegrees=2}
 {bone1Name=dyn_belt_r_03, bone2Name=dyn_belt_r_02, swing1LimitDegrees=20, swing2LimitDegrees=35, twistLimitDegrees=2}
 {bone1Name=dyn_belt_l_01, bone2Name=dyn_belt_l_MSTR, swing1LimitDegrees=2, swing2LimitDegrees=10, twistLimitDegrees=2}
 {bone1Name=dyn_belt_l_02, bone2Name=dyn_belt_l_01, swing1LimitDegrees=2, swing2LimitDegrees=10, twistLimitDegrees=2}
 {bone1Name=dyn_belt_l_03, bone2Name=dyn_belt_l_02, swing1LimitDegrees=2, swing2LimitDegrees=10, twistLimitDegrees=2}
 {
  bone1Name=dyn_belt_r_MSTR
  bone2Name=dyn_belt_MSTR
  swing1LimitDegrees=45
  swing1Motion=Free
  swing2LimitDegrees=45
  swing2Motion=Free
  twistLimitDegrees=45
  twistMotion=Free
 }
 {
  bone1Name=dyn_belt_l_MSTR
  bone2Name=dyn_belt_MSTR
  swing1LimitDegrees=45
  swing1Motion=Free
  swing2LimitDegrees=45
  swing2Motion=Free
  twistLimitDegrees=45
  twistMotion=Free
 }
]
