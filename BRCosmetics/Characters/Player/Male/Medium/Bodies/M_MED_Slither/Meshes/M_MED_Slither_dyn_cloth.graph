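# Physics asset for the Slither male medium player body, simulating the dynamic back skirt and belt pieces of the costume.
# Capsule chains for three back skirt strips and left and right belt strands root at small MSTR bodies linked to the pelvis, the belt via a shared master.
# Skirt links swing mainly on one axis with tight twist, while belt links swing moderately on both.
# Pelvis and thigh capsules provide body volumes.
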PhysicsAsset /BRCosmetics/Characters/Player/Male/Medium/Bodies/M_MED_Slither/Meshes/M_MED_Slither_dyn_cloth
bodies {
 dyn_belt_MSTR [{shapeType=Capsule, length=0.505, radius=0.505, shapeName=dyn_skirt_bk_r_MSTR_capsule, rotation=<rot 0 -90 0>}]
 dyn_belt_l_01 [{shapeType=Capsule, length=3, radius=2, shapeName=dyn_skirt_bk_l_01_capsule, center=<3 0 0>, rotation=<rot 90 0 0>}]
 dyn_belt_l_02 [{shapeType=Capsule, length=3, radius=2, shapeName=dyn_skirt_bk_l_01_capsule, center=<3 0 0>, rotation=<rot 90 0 0>}]
 dyn_belt_l_03 [{shapeType=Capsule, length=3, radius=2, shapeName=dyn_skirt_bk_l_02_capsule, center=<3 0 0>, rotation=<rot 90 0 0>}]
 dyn_belt_l_MSTR [{shapeType=Capsule, length=0.505, radius=0.505, shapeName=dyn_skirt_bk_r_MSTR_capsule, rotation=<rot 0 -90 0>}]
 dyn_belt_r_01 [{shapeType=Capsule, length=5, radius=2, shapeName=dyn_skirt_bk_l_01_capsule, center=<4 0 0>, rotation=<rot 90 0 0>}]
 dyn_belt_r_02 [{shapeType=Capsule, length=2.5, radius=3, shapeName=dyn_skirt_bk_l_01_capsule, center=<4 0 0>, rotation=<rot 90 0 0>}]
 dyn_belt_r_03 [{shapeType=Capsule, length=2.5, radius=3, shapeName=dyn_skirt_bk_l_02_capsule, center=<4 0 0>, rotation=<rot 90 0 0>}]
 dyn_belt_r_MSTR [{shapeType=Capsule, length=0.505, radius=0.505, shapeName=dyn_skirt_bk_r_MSTR_capsule, rotation=<rot 0 -90 0>}]
 dyn_skirt_bk_01 [{shapeType=Capsule, length=18.3339, radius=4.6715, shapeName=dyn_skirt_bk_l_01_capsule, center=<3 0 5>, rotation=<rot 0 0 90>}]
 dyn_skirt_bk_02 [{shapeType=Capsule, length=17.5839, radius=5.4215, shapeName=dyn_skirt_bk_l_02_capsule, center=<3 0 5>, rotation=<rot 0 0 90>}]
 dyn_skirt_bk_MSTR [{shapeType=Capsule, length=0.505, radius=0.505, shapeName=dyn_skirt_bk_r_MSTR_capsule, rotation=<rot 0 -90 0>}]
 dyn_skirt_bk_base [{shapeType=Capsule, length=18.0839, radius=4.9215, shapeName=dyn_skirt_bk_l_01_capsule, center=<3 0 5>, rotation=<rot 0 0 90>}]
 dyn_skirt_bk_l_01 [{shapeType=Capsule, length=13.25, radius=5.25, shapeName=dyn_skirt_bk_l_01_capsule, center=<3 0 5>, rotation=<rot 0 0 79.9998>}]
 dyn_skirt_bk_l_02 [{shapeType=Capsule, length=13.25, radius=5.25, shapeName=dyn_skirt_bk_l_02_capsule, center=<3 0 5>, rotation=<rot 0 0 79.9998>}]
 dyn_skirt_bk_l_MSTR [{shapeType=Capsule, length=0.505, radius=0.505, shapeName=dyn_skirt_bk_l_MSTR_capsule, rotation=<rot 0 -90 0>}]
 dyn_skirt_bk_l_base [{shapeType=Capsule, length=13.5, radius=5, shapeName=dyn_skirt_bk_l_01_capsule, center=<3 0 5>, rotation=<rot 0 0 79.9997>}]
 dyn_skirt_bk_r_01 [{shapeType=Capsule, length=13.25, radius=5.25, shapeName=dyn_skirt_bk_l_01_capsule, center=<3 0 5>, rotation=<rot 0 0 100>}]
 dyn_skirt_bk_r_02 [{shapeType=Capsule, length=13.25, radius=5.25, shapeName=dyn_skirt_bk_l_02_capsule, center=<3 0 5>, rotation=<rot 0 0 100>}]
 dyn_skirt_bk_r_MSTR [{shapeType=Capsule, length=0.505, radius=0.505, shapeName=dyn_skirt_bk_r_MSTR_capsule, rotation=<rot 0 -90 0>}]
 dyn_skirt_bk_r_base [{shapeType=Capsule, length=13.5, radius=5, shapeName=dyn_skirt_bk_l_01_capsule, center=<3 0 5>, rotation=<rot 0 0 100>}]
 pelvis [{shapeType=Capsule, length=8.5, radius=10.5232, shapeName=pelvis_capsule, center=<0.5792 -1.5317 0>}]
 thigh_twist_01_l [{
   shapeType=Capsule
   length=35.932
   radius=11.2357
   shapeName=thigh_twist_01_l_capsule
   center=<5.2089 -1.1101 3.249>
   rotation=<rot 90 0 -1.9773>
  }]
 thigh_twist_01_r [{
   shapeType=Capsule
   length=35.932
   radius=11.2357
   shapeName=thigh_twist_01_l_capsule
   center=<-5.2089 1.1101 -3.249>
   rotation=<rot 90 0 178.0227>
  }]
}
constraints [
 {bone1Name=dyn_skirt_bk_l_01, bone2Name=dyn_skirt_bk_l_base, swing1LimitDegrees=3, swing2LimitDegrees=45, twistLimitDegrees=1}
 {bone1Name=dyn_skirt_bk_l_02, bone2Name=dyn_skirt_bk_l_01, swing1LimitDegrees=3, swing2LimitDegrees=45, twistLimitDegrees=1}
 {bone1Name=dyn_skirt_bk_r_01, bone2Name=dyn_skirt_bk_r_base, swing1LimitDegrees=3, swing2LimitDegrees=45, twistLimitDegrees=1}
 {bone1Name=dyn_skirt_bk_r_02, bone2Name=dyn_skirt_bk_r_01, swing1LimitDegrees=3, swing2LimitDegrees=45, twistLimitDegrees=1}
 {bone1Name=dyn_skirt_bk_02, bone2Name=dyn_skirt_bk_01, swing1LimitDegrees=3, swing2LimitDegrees=45, twistLimitDegrees=1}
 {bone1Name=dyn_skirt_bk_01, bone2Name=dyn_skirt_bk_base, swing1LimitDegrees=3, swing2LimitDegrees=45, twistLimitDegrees=1}
 {
  bone1Name=dyn_skirt_bk_MSTR
  bone2Name=pelvis
  swing1LimitDegrees=45
  swing1Motion=Free
  swing2LimitDegrees=45
  swing2Motion=Free
  twistLimitDegrees=45
  twistMotion=Free
 }
 {
  bone1Name=dyn_skirt_bk_l_MSTR
  bone2Name=pelvis
  swing1LimitDegrees=45
  swing1Motion=Free
  swing2LimitDegrees=45
  swing2Motion=Free
  twistLimitDegrees=45
  twistMotion=Free
 }
 {
  bone1Name=dyn_skirt_bk_r_MSTR
  bone2Name=pelvis
  swing1LimitDegrees=45
  swing1Motion=Free
  swing2LimitDegrees=45
  swing2Motion=Free
  twistLimitDegrees=45
  twistMotion=Free
 }
 {bone1Name=dyn_skirt_bk_l_base, bone2Name=dyn_skirt_bk_l_MSTR, swing1LimitDegrees=3, swing2LimitDegrees=45, twistLimitDegrees=1}
 {bone1Name=dyn_skirt_bk_r_base, bone2Name=dyn_skirt_bk_r_MSTR, swing1LimitDegrees=3, swing2LimitDegrees=45, twistLimitDegrees=1}
 {bone1Name=dyn_skirt_bk_base, bone2Name=dyn_skirt_bk_MSTR, swing1LimitDegrees=3, swing2LimitDegrees=45, twistLimitDegrees=1}
 {
  bone1Name=thigh_twist_01_r
  bone2Name=pelvis
  swing1LimitDegrees=45
  swing1Motion=Free
  swing2LimitDegrees=45
  swing2Motion=Free
  twistLimitDegrees=45
  twistMotion=Free
 }
 {
  bone1Name=thigh_twist_01_l
  bone2Name=pelvis
  swing1LimitDegrees=45
  swing1Motion=Free
  swing2LimitDegrees=45
  swing2Motion=Free
  twistLimitDegrees=45
  twistMotion=Free
 }
 {
  bone1Name=dyn_belt_MSTR
  bone2Name=pelvis
  swing1LimitDegrees=45
  swing1Motion=Free
  swing2LimitDegrees=45
  swing2Motion=Free
  twistLimitDegrees=45
  twistMotion=Free
 }
 {bone1Name=dyn_belt_r_01, bone2Name=dyn_belt_r_MSTR, swing1LimitDegrees=30, swing2LimitDegrees=30, twistLimitDegrees=2}
 {bone1Name=dyn_belt_r_02, bone2Name=dyn_belt_r_01, swing1LimitDegrees=30, swing2LimitDegrees=30, twistLimitDegrees=2}
 {bone1Name=dyn_belt_r_03, bone2Name=dyn_belt_r_02, swing1LimitDegrees=30, swing2LimitDegrees=30, twistLimitDegrees=2}
 {bone1Name=dyn_belt_l_01, bone2Name=dyn_belt_l_MSTR, swing1LimitDegrees=30, swing2LimitDegrees=30, twistLimitDegrees=2}
 {bone1Name=dyn_belt_l_02, bone2Name=dyn_belt_l_01, swing1LimitDegrees=30, swing2LimitDegrees=30, twistLimitDegrees=2}
 {bone1Name=dyn_belt_l_03, bone2Name=dyn_belt_l_02, swing1LimitDegrees=30, swing2LimitDegrees=30, twistLimitDegrees=2}
 {
  bone1Name=dyn_belt_r_MSTR
  bone2Name=dyn_belt_MSTR
  swing1LimitDegrees=45
  swing1Motion=Free
  swing2LimitDegrees=45
  swing2Motion=Free
  twistLimitDegrees=45
  twistMotion=Free
 }
 {
  bone1Name=dyn_belt_l_MSTR
  bone2Name=dyn_belt_MSTR
  swing1LimitDegrees=45
  swing1Motion=Free
  swing2LimitDegrees=45
  swing2Motion=Free
  twistLimitDegrees=45
  twistMotion=Free
 }
]
