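# Animation Blueprint for the Slither medium male player body, adding deformation and secondary motion to the pose copied from its attached parent mesh.
# A deform Control Rig feeds bone constraints on the belt and front skirt collision, then additive belt offsets weighted by thigh-raise values the event graph derives each update.
# AnimDynamics then swings the front skirt bones, and RigidBody simulates the belt; the cloth RigidBody has zero alpha.
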AnimBlueprint /BRCosmetics/Characters/Player/Male/Medium/Bodies/M_MED_Slither/Meshes/M_MED_Slither_AnimBP
target_skeleton=&/M_MED_Slither_Skeleton
graphs [
 AnimGraph {nodes [
   "Constraint - dyn_belt_MSTR (PPP)"!(ComponentPose, Alpha=1, ConstraintWeights_0=1, ConstraintWeights_1:float=0, ConstraintWeights_2:float=0) -> (Pose) {
    alphaBoolBlend {blendOption=Linear}
    boneToModify {boneName=dyn_belt_MSTR}
    constraintSetup [{targetBone {boneName=pelvis}}, {targetBone {boneName=spine_01}}, {targetBone {boneName=dyn_belt_ft_l}}]
    constraintWeights [1,1,1]
   }
   n1 = Control Rig!(
    Source=Copy Pose From Mesh() {bUseAttachedParent=true}
    Alpha
    KneeRot_fr_l_114deg:Vector
    KneeTrans_fr_l_114deg:Vector="0.258,-3.591,0.009"
    KneeScale_fr_l_114deg:Vector="1,1,1"
    KneeScale_fr_l_130deg:Vector="1,1,1"
    KneeRot_fr_l_130deg:Vector="0,0,-51.2"
    KneeTrans_fr_l_130deg:Vector="0.258,-3.591,0.009"
   ) -> (Pose) {
    alphaBoolBlend {blendOption=Linear}
    controlRigAssetReference {blueprintRigClass=&/M_MED_Slither_Deform_CtrlRig_sculpters.M_MED_Slither_Deform_CtrlRig_sculpters_C}
   }
   n2 = "Constraint - dyn_belt_MSTR (PPP)"(ComponentPose=Local To Component(LocalPose=@n1), Alpha=1, ConstraintWeights_0=0.6297, ConstraintWeights_1=0.2808, ConstraintWeights_2=0.0895) {
    alphaBoolBlend {blendOption=Linear}
    boneToModify {boneName=dyn_belt_MSTR}
    constraintSetup [{targetBone {boneName=pelvis}}, {targetBone {boneName=spine_01}}, {targetBone {boneName=dyn_belt_ft_l}}]
    constraintWeights [1,1,1]
   }
   n3 = "Constraint - dyn_skirt_ft_collision (PPP)"!(ComponentPose=@n2, Alpha=1, ConstraintWeights_0=0.4, ConstraintWeights_1:float=0.4, ConstraintWeights_2:float=0.2) -> (Pose) {
    alphaBoolBlend {blendOption=Linear}
    boneToModify {boneName=dyn_skirt_ft_collision}
    constraintSetup [{targetBone {boneName=thigh_l}}, {targetBone {boneName=thigh_r}}, {targetBone {boneName=pelvis}}]
    constraintWeights [1,1,1]
   }
   n4 = "Transform (Modify) Bone - Bone: dyn_belt_l_MSTR"(
    ComponentPose=@n3
    Alpha=Get thigh_rot_up_l() -> (thigh_rot_up_l:double)
    Translation="-0.391,-2.134,-1.970"
    Rotation="2.867,19.102,-34.899"
    Scale="1,1,1"
   ) {
    rotationMode=BMM_Additive
    rotationSpace=BCS_BoneSpace
    scaleMode=BMM_Additive
    scaleSpace=BCS_BoneSpace
    translationMode=BMM_Additive
    translationSpace=BCS_BoneSpace
    alphaBoolBlend {blendOption=Linear}
    boneToModify {boneName=dyn_belt_l_MSTR}
    rotation=<rot 2.867 19.102 -34.8991>
    translation=<-0.391 -2.134 -1.97>
   }
   n5 = Get thigh_rot_up_r() -> (thigh_rot_up_r:double)
   n6 = "Transform (Modify) Bone - Bone: dyn_belt_l_MSTR"(ComponentPose=@n4, Alpha=@n5, Translation="0,0,0", Rotation="0.000,0.000,-65", Scale="1,1,1") {
    rotationMode=BMM_Additive
    rotationSpace=BCS_BoneSpace
    scaleMode=BMM_Additive
    scaleSpace=BCS_BoneSpace
    translationMode=BMM_Additive
    translationSpace=BCS_BoneSpace
    alphaBoolBlend {blendOption=Linear}
    boneToModify {boneName=dyn_belt_l_MSTR}
    rotation {roll=-65}
   }
   n7 = "Transform (Modify) Bone - Bone: dyn_belt_r_MSTR"(ComponentPose=@n6, Alpha=@n5, Translation="-1.5,2.2,0.9", Rotation="-15,20,75", Scale="1,1,1") {
    rotationMode=BMM_Additive
    rotationSpace=BCS_BoneSpace
    scaleMode=BMM_Additive
    scaleSpace=BCS_BoneSpace
    translationMode=BMM_Additive
    translationSpace=BCS_BoneSpace
    alphaBoolBlend {blendOption=Linear}
    alphaScaleBiasClamp {interpSpeedIncreasing=30}
    boneToModify {boneName=dyn_belt_r_MSTR}
    rotation=<rot -15 20 75>
    translation=<-1.5 2.2 0.9>
   }
   n8 = Anim Dynamics - Bone: dyn_skirt_ft_r(ComponentPose=@n7, Alpha=1, ExternalForce="0,0,0") {
    angularDampingOverride=0.9999
    bGravityOverrideInSimSpace=true
    bOverrideAngularDamping=true
    bUseGravityOverride=true
    linearDampingOverride=0.9
    simulationSpace=RootRelative
    alphaBoolBlend {blendOption=Linear}
    alphaScaleBiasClamp {scale=0.6}
    boundBone {boneName=dyn_skirt_ft_r}
    componentAppliedLinearAccClamp {z=5000}
    componentLinearAccScale {z=0.4}
    componentLinearVelScale {z=0.6}
    gravityOverride {z=-980}
    physicsBodyDefinitions [{
      sphereCollisionRadius=0
      boundBone {boneName=dyn_skirt_ft_r}
      boxExtents=<10 15 20>
      constraintSetup {angularLimitsMax=<0 0 0>, angularLimitsMin=<0 0 -35>, angularTarget=<0 0 0>}
      localJointOffset=<5 0 0>
     }]
   }
   n9 = Anim Dynamics - Bone: dyn_skirt_ft_l(ComponentPose=@n8, Alpha=1, ExternalForce="0,0,0") {
    angularDampingOverride=0.9999
    bGravityOverrideInSimSpace=true
    bOverrideAngularDamping=true
    bUseGravityOverride=true
    linearDampingOverride=0.9
    simulationSpace=RootRelative
    alphaBoolBlend {blendOption=Linear}
    alphaScaleBiasClamp {scale=0.6}
    boundBone {boneName=dyn_skirt_ft_l}
    componentAppliedLinearAccClamp {z=5000}
    componentLinearAccScale {z=0.4}
    componentLinearVelScale {z=0.6}
    gravityOverride {z=-980}
    physicsBodyDefinitions [{
      sphereCollisionRadius=0
      boundBone {boneName=dyn_skirt_ft_l}
      boxExtents=<10 15 20>
      constraintSetup {angularLimitsMax=<0 0 0>, angularLimitsMin=<0 0 -35>, angularTarget=<0 0 0>}
      localJointOffset=<5 0 0>
     }]
   }
   n10 = RigidBody!(ComponentPose=@n9, Alpha=0, OverrideWorldGravity:Vector="0,0,-980", ExternalForce) -> (Pose) {
    bClampLinearTranslationLimitToRefPose=true
    bOverrideWorldGravity=true
    overridePhysicsAsset=&/M_MED_Slither_dyn_cloth
    simulationSpace=BaseBoneSpace
    alphaBoolBlend {blendOption=Linear}
    baseBoneRef {boneName=spine_01}
    componentAppliedLinearAccClamp=<3000 3000 3000>
    componentLinearAccScale {x=0.45}
    componentLinearVelScale {x=0.45}
    overrideWorldGravity {z=-980}
    simSpaceSettings {
     maxAngularAcceleration=200
     maxAngularVelocity=5
     maxLinearAcceleration=1000
     maxLinearVelocity=200
     velocityScaleZ=0
     worldAlpha=1
     externalLinearDragV=<1 1>
    }
   }
   n11 = RigidBody(ComponentPose=@n10, OverrideWorldGravity="0,0,-980") {
    bClampLinearTranslationLimitToRefPose=true
    bOverrideWorldGravity=true
    overridePhysicsAsset=&/M_MED_Slither_dyn_belt
    simulationSpace=BaseBoneSpace
    alphaBoolBlend {blendOption=Linear}
    baseBoneRef {boneName=spine_01}
    componentAppliedLinearAccClamp=<3000 3000 3000>
    componentLinearAccScale {x=0.45}
    componentLinearVelScale {x=0.45}
    overrideWorldGravity {z=-980}
    simSpaceSettings {
     maxAngularAcceleration=200
     maxAngularVelocity=5
     maxLinearAcceleration=1000
     maxLinearVelocity=200
     velocityScaleZ=0
     worldAlpha=1
     externalLinearDragV=<1 1>
    }
   }
   Output Pose(Result=Component To Local(ComponentPose=@n11)) {layerGroup=None}
  ]}
 EventGraph {nodes [
   Event BlueprintUpdateAnimation()
   > n1 = Sequence()
   n2 = GetOwnerSkeletalMeshComponent()
   n3 = BreakRotator(InRot=BreakTransform(InTransform=GetDeltaTransformFromRefPose(self=@n2, BoneName=thigh_r, BaseName=spine_01)).Rotation)
   > Set thigh_rot_in_r(thigh_rot_in_r:double=MapRangeClamped(Value=@n3.Pitch, InRangeA=0, InRangeB=-30)) -> (Output_Get:double)
   > Set thigh_rot_up_r(thigh_rot_up_r:double=MapRangeClamped(Value=@n3.Yaw, InRangeA=-30, InRangeB=-120)) -> (Output_Get:double)
   > Set thigh_rot_bk_r(thigh_rot_bk_r:double=MapRangeClamped(Value=@n3.Yaw, InRangeA=-15, InRangeB=10, OutRangeA=1, OutRangeB=0)) -> (Output_Get:double)
   n4 = GetOwnerSkeletalMeshComponent()
   n5 = BreakRotator(InRot=BreakTransform(InTransform=GetDeltaTransformFromRefPose(self=@n4, BoneName=thigh_l, BaseName=spine_01)).Rotation)
   > Set thigh_rot_in_l(thigh_rot_in_l:double=MapRangeClamped(Value=@n5.Pitch, InRangeA=0, InRangeB=-20)) -> (Output_Get:double)
   > Set thigh_rot_out_l(thigh_rot_out_l:double=MapRangeClamped(Value=@n5.Pitch, InRangeA=0, InRangeB=30)) -> (Output_Get:double)
   > Set thigh_rot_up_l(thigh_rot_up_l:double=MapRangeClamped(Value=@n5.Yaw, InRangeA=-30, InRangeB=-130)) -> (Output_Get:double)
   > Set thigh_rot_bk_l(thigh_rot_bk_l:double=MapRangeClamped(Value=@n5.Yaw, InRangeA=0, InRangeB=20)) -> (Output_Get:double)
   n6 = Get thigh_rot_up_r()
   n7 = Get thigh_rot_up_l()
   Set fornt_skirt_aim_alpha(execute=@n1.then_1, fornt_skirt_aim_alpha:double="Max (Float)"(A=@n6, B=@n7)) -> (Output_Get:double)
   float * float(A=float + float(A=@n6, B=@n7), B=0.5) {is=title K2Node_PromotableOperator}
  ]}
]
variables [
 fornt_skirt_aim_alpha {type=number}
 thigh_rot_bk_l {type=number}
 thigh_rot_bk_r {type=number}
 thigh_rot_in_l {type=number}
 thigh_rot_in_r {type=number}
 thigh_rot_out_l {type=number}
 thigh_rot_up_l {type=number}
 thigh_rot_up_r {type=number}
]
